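# Animation Blueprint for the Lurk glider cosmetic, animating its opening, gliding and closed poses from skydiving state.
# A Closed/Open Into/Idle Open state machine follows parachute-open or force-open flags, with front-end animations for cosmetic preview.
# Additives layer acceleration lean (inverted ParaGlide_lean curve alpha), aim-driven turn, wing jitter, and turbine spin, slower backward and faster forward, before DefaultSlot.
# The TurnONEngines notify marks the glider fully deployed.
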AnimBlueprint /BRCosmetics/Animation/Game/MainPlayer/Skydive/ParaGlide/Lurk/Lurk_Glider_Anim_BP
target_skeleton=/BRCosmetics/Gadgets/Assets/VinderTech_GliderChute/Glider_Sienna_Male/Meshes/Sienna_Male_Glider_Skeleton
graphs [
 AnimGraph {nodes [
   n1 = Break Fort Anim Input Skydiving!(FortAnimInput_Skydiving:"Fort Anim Input Skydiving Structure (by ref)"=Get SkydivingState() -> (SkydivingState:Fort Anim Input Skydiving Structure)) -> (
    ParaGilderRootOffset:Vector=~
    ParaGliderRootRotationOverride:Rotator=~
    bUseParaGlideRootModifier:bool=~
    bIsSkydivingFromLaunchPad:bool=~
    bIsSkydivingFromBus:bool=~
    bIsInVortex:bool=~
    bIsDiving:bool=~
    bIsParachuteOpen:bool=~
    bIsSkydiving:bool=~
    bIsParachuteLeaning:bool=~
    bIsSkydiveLeaning:bool=~
    bIsLeaning:bool=~
    bIsSkydiveDiveMode:bool=~
    bPlaySkydiveDrift:bool=~
    bSkydiveDriftDelayActive:bool=~
    bSkydiveDriftAnimAllowed:bool=~
    bIsGliderRight:bool=~
    bIsGliderCenter:bool=~
    bIsGliderLeft:bool=~
    bIsGliderForward:bool=~
    bIsGliderBack:bool=~
    bHasDeployedAndIsStandingUpright:bool=~
    LocalAccelForward:float=~
    LocalAccelRight:float=~
    LocalVelocityRight:float=~
    SkydiveAimPitch:float=~
    SkydiveAimYaw:float=~
    DeployChuteAnimRate:float=~
    SkydiveDriftAnimRate:float=~
    SkydiveDriftAnimRateCurrent:float=~
    SkydiveFidgetAnimRate:float=~
    SkydiveFidgetAnimRateCurrent:float=~
    SkydiveAdditiveAlpha:float=~
    SkydiveDriftDelay:float=~
    SkydiveDriftAnim:int
    SkydiveDriftAnimMax:int
    LaunchpadAnim:int
    LocalAccelDir:ESkydivingDirection Enum
    DirectionLast:ESkydivingDirection Enum
    PoseAdditive:Anim Sequence Object Reference
   ) {is=title K2Node_BreakStruct}
   n2 = Blendspace Player 'Lurk_Glider_Lean_BS'(X=@n1.LocalAccelRight, Y=@n1.LocalAccelForward) {blendSpace=&/Lurk_Glider_Lean_BS}
   AnimGraphNode_StateMachine_0 = Locomotion() {is=title AnimGraphNode_StateMachine, bSkipFirstUpdateTransition=false}
   n3 = Break Fort Anim Input Skydiving(FortAnimInput_Skydiving=Get SkydivingState()) -> (
    ParaGilderRootOffset=~
    ParaGliderRootRotationOverride=~
    bUseParaGlideRootModifier=~
    bIsSkydivingFromLaunchPad=~
    bIsSkydivingFromBus=~
    bIsInVortex=~
    bIsDiving=~
    bIsParachuteOpen=~
    bIsSkydiving=~
    bIsParachuteLeaning=~
    bIsSkydiveLeaning=~
    bIsLeaning=~
    bIsSkydiveDiveMode=~
    bPlaySkydiveDrift=~
    bSkydiveDriftDelayActive=~
    bSkydiveDriftAnimAllowed=~
    bIsGliderRight=~
    bIsGliderCenter=~
    bIsGliderLeft=~
    bIsGliderForward=~
    bIsGliderBack=~
    bHasDeployedAndIsStandingUpright=~
    LocalAccelForward=~
    LocalAccelRight=~
    LocalVelocityRight=~
    SkydiveAimPitch=~
    SkydiveAimYaw=~
    DeployChuteAnimRate=~
    SkydiveDriftAnimRate=~
    SkydiveDriftAnimRateCurrent=~
    SkydiveFidgetAnimRate=~
    SkydiveFidgetAnimRateCurrent=~
    SkydiveAdditiveAlpha=~
    SkydiveDriftDelay=~
   )
   n4 = Apply Additive(
    Base=Apply Additive(
     Base=Apply Additive!(Base=@AnimGraphNode_StateMachine_0, Additive=@n2, AlphaCurveName:Name=ParaGlide_lean) -> (Pose) {alphaInputType=Curve, alphaBoolBlend {blendOption=Linear}, alphaScaleBiasClamp {bias=1, scale=-1}}
     Additive=Blendspace Player 'Lurk_Glider_Turn_BS'(X=@n3.SkydiveAimYaw, Y=@n3.SkydiveAimPitch) {blendSpace=&/Lurk_Glider_Turn_BS}
    ) {alphaBoolBlend {blendOption=Linear}}
    Additive="Sequence Player (additive) 'Lurk_WingJitter_GLIDER'"() {sequence=&/Lurk_WingJitter_GLIDER}
   ) {alphaBoolBlend {blendOption=Linear}}
   n5 = Blend Poses by bool(
    BlendPose_0="Sequence Player (additive) 'Lurk_TurbineSpin_GLIDER'"() {playRate=0.5, sequence=&/Lurk_TurbineSpin_GLIDER}
    BlendPose_1=Blend Poses by bool(
     BlendPose_0="Sequence Player (additive) 'Lurk_TurbineSpin_GLIDER'"() {playRate=1.5, sequence=&Lurk_TurbineSpin_GLIDER}
     BlendPose_1="Sequence Player (additive) 'Lurk_TurbineSpin_GLIDER'"() {sequence=&Lurk_TurbineSpin_GLIDER}
     bActiveValue=@n3.bIsGliderForward
    ) {blendType=Linear}
    bActiveValue=@n3.bIsGliderBack
   ) {blendType=Linear}
   Output Pose(Result=Slot 'DefaultSlot'(Source=Apply Additive!(Base=@n4, Additive=@n5, Alpha:float) -> (Pose) {alphaBoolBlend {blendOption=Linear}})) {layerGroup=None}
  ]}
 Locomotion {
  entry=Closed
  path=AnimGraph.AnimGraphNode_StateMachine_0.Locomotion
  ids=id
  nodes [
   AnimStateNode_0 = Closed() {is=title AnimStateNode}
   AnimStateNode_1 = Open Into() {is=title AnimStateNode}
   AnimStateNode_2 = Idle Open() {is=title AnimStateNode}
   AnimStateTransitionNode_0 = AnimStateTransitionNode() {crossfadeDuration=0}
   AnimStateTransitionNode_1 = AnimStateTransitionNode() {bAutomaticRuleBasedOnSequencePlayerInState=true}
   AnimStateTransitionNode_2 = AnimStateTransitionNode() {crossfadeDuration=0}
   AnimStateTransitionNode_3 = AnimStateTransitionNode()
  ]
  links [
   {
    dst=Open Into
    rule=AnimGraph.AnimGraphNode_StateMachine_0.Locomotion.AnimStateTransitionNode_3.Transition
    src=Closed
    via=AnimStateTransitionNode_3
   }
   {
    dst=Closed
    rule=AnimGraph.AnimGraphNode_StateMachine_0.Locomotion.AnimStateTransitionNode_0.Transition
    src=Idle Open
    via=AnimStateTransitionNode_0
   }
   {
    dst=Closed
    rule=AnimGraph.AnimGraphNode_StateMachine_0.Locomotion.AnimStateTransitionNode_2.Transition
    src=Open Into
    via=AnimStateTransitionNode_2
   }
   {
    dst=Idle Open
    rule=AnimGraph.AnimGraphNode_StateMachine_0.Locomotion.AnimStateTransitionNode_1.Transition
    src=Open Into
    via=AnimStateTransitionNode_1
   }
  ]
 }
 Closed {
  path=AnimGraph.AnimGraphNode_StateMachine_0.Locomotion.AnimStateNode_0.Closed
  nodes [Output Animation Pose(Result=Sequence Player 'Lurk_Closed_GLIDER'() {sequence=&/Lurk_Closed_GLIDER})]
 }
 Open Into {
  path=AnimGraph.AnimGraphNode_StateMachine_0.Locomotion.AnimStateNode_1.Open Into
  nodes [
   n1 = Get OwnerGlider() -> (OwnerGlider:Fort Player Parachute Object Reference)
   Output Animation Pose(Result=Blend Poses by bool(
     BlendPose_0=Sequence Player 'Lurk_Into_FrontEnd2_GLIDER'!(bLoopAnimation:bool=False) -> (Pose) {bLoopAnimation=false, sequence=&/Lurk_Into_FrontEnd2_GLIDER}
     BlendPose_1=Sequence Player 'Lurk_Into_GLIDER'!(bLoopAnimation:bool=False) -> (Pose) {bLoopAnimation=false, sequence=&/Lurk_Into_GLIDER}
     bActiveValue=Get IsCosmeticPreview!(self:Fort Player Parachute Object Reference=@n1) -> (IsCosmeticPreview:bool)
    ) {blendType=Linear})
  ]
 }
 Idle Open {
  path=AnimGraph.AnimGraphNode_StateMachine_0.Locomotion.AnimStateNode_2.Idle Open
  nodes [
   n1 = Get OwnerGlider()
   Output Animation Pose(Result=Blend Poses by bool(
     BlendPose_0=Sequence Player 'Lurk_Glide_FrontEnd_GLIDER'() {sequence=&/Lurk_Glide_FrontEnd_GLIDER}
     BlendPose_1=Sequence Player 'Lurk_Glide_GLIDER'() {sequence=&/Lurk_Glide_GLIDER}
     bActiveValue=Get IsCosmeticPreview(self=@n1)
    ) {blendType=Linear})
  ]
 }
 Transition {
  path=AnimGraph.AnimGraphNode_StateMachine_0.Locomotion.AnimStateTransitionNode_0.Transition
  nodes [
   n1 = Break Fort Anim Input Skydiving(FortAnimInput_Skydiving=Get SkydivingState()) -> (
    ParaGilderRootOffset=~
    ParaGliderRootRotationOverride=~
    bUseParaGlideRootModifier=~
    bIsSkydivingFromLaunchPad=~
    bIsSkydivingFromBus=~
    bIsInVortex=~
    bIsDiving=~
    bIsParachuteOpen=~
    bIsSkydiving=~
    bIsParachuteLeaning=~
    bIsSkydiveLeaning=~
    bIsLeaning=~
    bIsSkydiveDiveMode=~
    bPlaySkydiveDrift=~
    bSkydiveDriftDelayActive=~
    bSkydiveDriftAnimAllowed=~
    bIsGliderRight=~
    bIsGliderCenter=~
    bIsGliderLeft=~
    bIsGliderForward=~
    bIsGliderBack=~
    bHasDeployedAndIsStandingUpright=~
    LocalAccelForward=~
    LocalAccelRight=~
    LocalVelocityRight=~
    SkydiveAimPitch=~
    SkydiveAimYaw=~
    DeployChuteAnimRate=~
    SkydiveDriftAnimRate=~
    SkydiveDriftAnimRateCurrent=~
    SkydiveFidgetAnimRate=~
    SkydiveFidgetAnimRateCurrent=~
    SkydiveAdditiveAlpha=~
    SkydiveDriftDelay=~
   )
   n2 = Get bForceOpen() -> (bForceOpen:bool)
   Result(bCanEnterTransition=AND Boolean(A=NOT Boolean(A=@n1.bIsParachuteOpen), B=NOT Boolean(A=@n2)))
  ]
 }
 Transition {path=AnimGraph.AnimGraphNode_StateMachine_0.Locomotion.AnimStateTransitionNode_1.Transition, nodes [Result()]}
 Transition {
  path=AnimGraph.AnimGraphNode_StateMachine_0.Locomotion.AnimStateTransitionNode_2.Transition
  nodes [
   n1 = Get bForceOpen()
   n2 = Break Fort Anim Input Skydiving(FortAnimInput_Skydiving=Get SkydivingState()) -> (
    ParaGilderRootOffset=~
    ParaGliderRootRotationOverride=~
    bUseParaGlideRootModifier=~
    bIsSkydivingFromLaunchPad=~
    bIsSkydivingFromBus=~
    bIsInVortex=~
    bIsDiving=~
    bIsParachuteOpen=~
    bIsSkydiving=~
    bIsParachuteLeaning=~
    bIsSkydiveLeaning=~
    bIsLeaning=~
    bIsSkydiveDiveMode=~
    bPlaySkydiveDrift=~
    bSkydiveDriftDelayActive=~
    bSkydiveDriftAnimAllowed=~
    bIsGliderRight=~
    bIsGliderCenter=~
    bIsGliderLeft=~
    bIsGliderForward=~
    bIsGliderBack=~
    bHasDeployedAndIsStandingUpright=~
    LocalAccelForward=~
    LocalAccelRight=~
    LocalVelocityRight=~
    SkydiveAimPitch=~
    SkydiveAimYaw=~
    DeployChuteAnimRate=~
    SkydiveDriftAnimRate=~
    SkydiveDriftAnimRateCurrent=~
    SkydiveFidgetAnimRate=~
    SkydiveFidgetAnimRateCurrent=~
    SkydiveAdditiveAlpha=~
    SkydiveDriftDelay=~
   )
   Result(bCanEnterTransition=AND Boolean(A=NOT Boolean(A=@n2.bIsParachuteOpen), B=NOT Boolean(A=@n1)))
  ]
 }
 Transition {
  path=AnimGraph.AnimGraphNode_StateMachine_0.Locomotion.AnimStateTransitionNode_3.Transition
  nodes [
   n1 = Get SkydivingState()
   Result(bCanEnterTransition=OR Boolean(
     A=Break Fort Anim Input Skydiving(FortAnimInput_Skydiving=@n1) -> (
      ParaGilderRootOffset=~
      ParaGliderRootRotationOverride=~
      bUseParaGlideRootModifier=~
      bIsSkydivingFromLaunchPad=~
      bIsSkydivingFromBus=~
      bIsInVortex=~
      bIsDiving=~
      bIsParachuteOpen=~
      bIsSkydiving=~
      bIsParachuteLeaning=~
      bIsSkydiveLeaning=~
      bIsLeaning=~
      bIsSkydiveDiveMode=~
      bPlaySkydiveDrift=~
      bSkydiveDriftDelayActive=~
      bSkydiveDriftAnimAllowed=~
      bIsGliderRight=~
      bIsGliderCenter=~
      bIsGliderLeft=~
      bIsGliderForward=~
      bIsGliderBack=~
      bHasDeployedAndIsStandingUpright=~
      LocalAccelForward=~
      LocalAccelRight=~
      LocalVelocityRight=~
      SkydiveAimPitch=~
      SkydiveAimYaw=~
      DeployChuteAnimRate=~
      SkydiveDriftAnimRate=~
      SkydiveDriftAnimRateCurrent=~
      SkydiveFidgetAnimRate=~
      SkydiveFidgetAnimRateCurrent=~
      SkydiveAdditiveAlpha=~
      SkydiveDriftDelay=~
     ).bIsParachuteOpen
     B=Get bForceOpen()
    ))
  ]
 }
 EventGraph {nodes [
   AnimNotify_TurnONEngines() {is=title K2Node_Event}
   > SetGliderFullyDeployed!(execute, self:Fort Glider Anim Instance Object Reference, bIsFullyDeployed:bool=true) -> (then)
   TryGetPawnOwner()
   Event BlueprintUpdateAnimation()
  ]}
]
variables []
